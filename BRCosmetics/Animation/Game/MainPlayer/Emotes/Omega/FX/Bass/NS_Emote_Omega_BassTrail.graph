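# Niagara system for the Omega emote's bass effect, drawing a soft smoke ribbon trail in a single emitter.
# Spawn rate follows a system-age curve that stops by two seconds; ribbons face along the owner's X axis and inherit source movement, damped vertically.
# Over each particle's life, color shifts red through purple to blue, width grows and brightness fades.
# Time-of-day scalars adjust color and intensity, while curves also drive dynamic material parameters.
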NiagaraSystem /BRCosmetics/Animation/Game/MainPlayer/Emotes/Omega/FX/Bass/NS_Emote_Omega_BassTrail
dependencies {
 usedDataInterfaces [&NiagaraDataInterfaceCurve, &NiagaraDataInterfaceColorCurve]
 usedDynamicInputs []
 usedModules [&SystemState]
}
emitters {Trail_Smoke_Soft {
  emitterUpdateScript [
   &EmitterState(Life Cycle Mode=::NewEnumerator1, Loop Behavior=::NewEnumerator1, Loop Delay=0.03, Loop Duration=2.15, UseLoopDelay=false)
   &SpawnRate(
    SpawnRate=&FloatFromCurve(FloatCurve=DataInterface() {propertyValues {Curve {keys [<at 0 1>,<at 1.6 1>,<at 2 0>]}}}, CurveIndex=$Engine.System.Age, Scale Curve=20)
    Use Spawn Probability=false
   )
  ]
  particleSpawnScript [
   &InitializeRibbon(
    Lifetime=1.4
    Module.Write Color=true
    Module.Write Lifetime=true
    Module.Write Mass=false
    Module.Write Material Random=true
    Module.Write Position=false
    Module.Write RibbonTwist=false
    Module.Write RibbonWidth=true
    Ribbon Width=60
   )
   /SetVariables(Particles.RibbonFacing=$Engine.Owner.SystemXAxis) {isSetParameters=true}
  ]
  particleUpdateScript [
   &ParticleState()
   &InheritSourceMovement(Applied Position Scale=<1 1 0.3>, ApplyForce=false, ApplyPosition=true, ApplyPositionFalloff=false)
   &SolveForcesAndVelocity(Clamp Velocity=false, Limit Acceleration=false)
   &Color(
    Color=&ColorFromCurve(ColorCurve=DataInterface() {propertyValues {
       AlphaCurve {keys [<at 0 1>,<at 0.25 1>,<at 1 0>]}
       BlueCurve {keys [
         {arriveTangent=4.0072, interpMode=RCIM_Cubic, time=0, value=0}
         <at 0.1526 0>
         <at 0.3189 0>
         <at 0.5036 1>
         <at 0.6607 1>
         <at 1 1 RCIM_Cubic>
        ]}
       GreenCurve {keys [
         {arriveTangent=-0.4007, interpMode=RCIM_Cubic, time=0, value=0.0833}
         <at 0.1526 0.0333>
         <at 0.3189 0.0333>
         <at 0.5036 0>
         <at 0.6607 0.2>
         <at 1 0.4 RCIM_Cubic>
        ]}
       RedCurve {keys [
         {arriveTangent=-11.8213, interpMode=RCIM_Cubic, time=0, value=1}
         <at 0.1526 1>
         <at 0.3189 1>
         <at 0.5036 0.6667>
         <at 0.6607 0.2>
         <at 1 0.1 RCIM_Cubic>
        ]}
      }})
    Scale Color=&Multiply_Vector(
     A=&VectorFromFloat(Value=&FNTimeOfDayScalar(Night Scale=0.5, Morning Scale=0.25, Evening Scale=0.5))
     B=&FN_Select_Vector_from_Time_Of_Day(Morning=<1 1 1>, Day=<1 1 1>, Evening=<1 1 1>, Night=<0.25 1 1>)
    )
   )
   &ScaleColor(
    Color Value To Scale=$Particles.Color
    Scale RGB=&VectorFromFloat(Value=&FloatFromCurve(
      FloatCurve=DataInterface() {propertyValues {Curve {keys [
          {interpMode=RCIM_Cubic, leaveTangent=-5.4692, tangentMode=RCTM_Break, time=0, value=2}
          {arriveTangent=0.1008, interpMode=RCIM_Cubic, leaveTangent=0.1008, tangentMode=RCTM_User, time=0.3, value=1}
          {interpMode=RCIM_Cubic, tangentMode=RCTM_User, time=1, value=0}
         ]}}}
      Scale Curve=&FNTimeOfDayScalar(Night Scale=6, Morning Scale=4, Day Scale=3, Evening Scale=4)
     ))
    ScaleA=true
    ScaleRGB=true
   )
   &ScaleRibbonWidth(Ribbon Width Scale=&FloatFromCurve(FloatCurve=DataInterface() {propertyValues {Curve {keys [<at 0 1 RCIM_Cubic>, {arriveTangent=6.3317, interpMode=RCIM_Cubic, leaveTangent=6.3317, tangentMode=RCTM_User, time=1, value=3}]}}}))
   &DynamicMaterialParameters(
    Index 0 Param 1=&FloatFromCurve(FloatCurve=DataInterface() {propertyValues {Curve {keys [
         {arriveTangent=-2.4649, interpMode=RCIM_Cubic, leaveTangent=-2.4649, tangentMode=RCTM_User, time=0, value=1}
         {interpMode=RCIM_Cubic, tangentMode=RCTM_User, time=1, value=0}
        ]}}})
    Index 0 Param 2=6
    Index 0 Param 3=-0.05
    Index 0 Param 4=&FloatFromCurve(
     FloatCurve=DataInterface() {propertyValues {Curve {keys [
         <at 0 0 RCIM_Cubic>
         {interpMode=RCIM_Cubic, tangentMode=RCTM_User, time=0.25, value=1}
         {arriveTangent=-0.6219, interpMode=RCIM_Cubic, leaveTangent=-0.6219, tangentMode=RCTM_User, time=1, value=0}
        ]}}}
     CurveIndex=$Emitter.NormalizedLoopAge
    )
    Module.Index0Float0Write=true
    Module.Index0Float1Write=true
    Module.Index0Float2Write=true
    Module.Index0Float3Write=true
   )
  ]
  rendererClasses [&NiagaraRibbonRendererProperties]
  renderers [{rendererClass=&NiagaraRibbonRendererProperties}]
 }}
summary {
 systemName=NS_Emote_Omega_BassTrail
 emitters [{emitterName=Trail_Smoke_Soft, rendererClasses [&NiagaraRibbonRendererProperties]}]
 userVariables [User.RibbonIDOffset {type=NiagaraInt32}]
}
